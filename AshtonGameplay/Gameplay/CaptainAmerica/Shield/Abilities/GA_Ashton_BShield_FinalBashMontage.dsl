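; Gameplay ability for the final bash of the Captain America shield, playing its montage on activation.
; After calling the parent, it applies a DisableSwitch effect to the owner to briefly block quickbar switching.
; It starts the montage from GetMontageToPlay and ends immediately without waiting, with stop-on-ability-end off so the animation can continue.
; CanApplyHolster is overridden to always return false.
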(event Ability|EventActivateAbility
  (|CallParentFunction)
  ;[0 0.8244 1] If player hits the final bash, ensure tthey can't switch quickbar for a short delay
  (Ability|ApplyGameplayEffectToOwner "/AshtonGameplay/Gameplay/CaptainAmerica/Shield/Effects/GE_Ashton_BShield_DisableSwitch.GE_Ashton_BShield_DisableSwitch_C")
  (Ability|Tasks|PlayMontageAndWait "None" (Variables|Default|GetMontageToPlay) 1.0 "None" false
    (:OnCompleted
      _)
    (:OnBlendOut
      _)
    (:OnInterrupted
      _)
    (:OnCancelled
      _))
  (Ability|EndAbility))

(fn CanApplyHolster
  (params (HolsterId :type Name%20%28by%20ref%29 :default "None"))
  (return false))
