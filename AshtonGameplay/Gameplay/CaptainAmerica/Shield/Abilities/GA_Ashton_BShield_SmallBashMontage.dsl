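; Gameplay ability for Captain America's shield that plays a small bash montage.
; On activation it runs the parent logic, forces a net update on the ability system owner, then waits for a client-server sync before playing the montage from GetMontageToPlay.
; The ability ends right after starting the montage, without waiting for it to finish.
; It overrides CanApplyHolster to always return false.
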(event Ability|EventActivateAbility
  (|CallParentFunction)
  (Networking|ForceNetUpdate (Components|GetOwner (Ability|GetAbilitySystemComponentfromActorInfo)))
  (Ability|Tasks|WaitNetSync "BothWait"
    (:OnSync
      (Ability|Tasks|PlayMontageAndWait "None" (Variables|Default|GetMontageToPlay) 1.0 "None" false
        (:OnCompleted
          _)
        (:OnInterrupted
          _)
        (:OnCancelled
          _))
      (Ability|EndAbility))))

(fn CanApplyHolster
  (params (HolsterId :type Name%20%28by%20ref%29 :default "None"))
  (return false))
