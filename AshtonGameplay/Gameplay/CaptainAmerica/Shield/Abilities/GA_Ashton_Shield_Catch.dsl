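; Gameplay ability for catching Captain America's shield, extending a parent catch ability.
; On activation it casts the avatar pawn's current weapon to B_Ashton_Shield_Weap and, if the player is trying to block, swaps in the catch-to-block montage before calling the parent activation.
; A failed avatar or weapon cast skips the parent activation.
; On end it calls the parent and pops the camera mode, though no push appears here.
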(event Ability|EventActivateAbility
  (bind _outfortplayerpawn (CallFunction|GetAvatarasValidFortPlayerPawn)
    (:ValidFortPlayerPawn
      (bind _asb_ashton_shield_weap (Utilities|Casting|CastToB_Ashton_Shield_Weap (Weapon|GetCurrentWeapon _outfortplayerpawn))
        (:then
          (if (Variables|/AshtonGameplay/Gameplay/CaptainAmerica/Shield/Actors/B_Ashton_Shield_Weap.B_Ashton_Shield_Weap_C|GetbTryingToBlock _asb_ashton_shield_weap)
            (Variables|CatchVars|SetMontageToPlay (Variables|Default|GetCatchToBlockMontage))
            (else
              _))
          (|CallParentFunction))
        (:CastFailed
          _)))
    (:AvatarCastFailed
      _)))

(event Ability|EventOnEndAbility
  (params (bWasCancelled :type Boolean))
  (|CallParentFunction bWasCancelled)
  (Camera|PopCameraMode))
