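; Passive Captain America shield ability that tracks whether the player holds block, even while the shield is thrown, so the catch can pick the right montage.
; On activation it caches pawn and weapon, then listens for secondary press and release gameplay events to set bTryingToBlock on the weapon.
; A press with no projectile out also has the weapon request a server cancel of tactical sprint.
; Ending the ability clears the flag.
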(event Ability|EventActivateAbility
  ;[0 1 0.2126] Check if the player is attempting to block. This GA is able to track the block input while the projectile is thrown, so we can determine if the player is holding block while the weapon is holstered. This allows us to determine which montage to play on catch.
  (|CallParentFunction)
  (bind _outfortplayerpawn (CallFunction|GetAvatarasValidFortPlayerPawn)
    (:ValidFortPlayerPawn
      (bind _output_get (Variables|Default|SetPlayerPawn _outfortplayerpawn))
      (bind _asb_ashton_shield_weap (Utilities|Casting|CastToB_Ashton_Shield_Weap (Weapon|GetCurrentWeapon _output_get)))
      (Variables|Default|SetWeapon _asb_ashton_shield_weap)
      (Ability|Tasks|WaitGameplayEvent "(TagName=\"Event.Secondary.Pressed\")"
        (:then
          (Ability|Tasks|WaitGameplayEvent "(TagName=\"Event.Secondary.Released\")"
            (:EventReceived
              (Variables|Default|GetPlayerPawn
                (:then
                  (bind _weapon (Variables|Default|GetWeapon)
                    (:then
                      (Variables|/AshtonGameplay/Gameplay/CaptainAmerica/Shield/Actors/B_Ashton_Shield_Weap.B_Ashton_Shield_Weap_C|SetbTryingToBlock false _weapon))
                    (:else)))
                (:else)))))
        (:EventReceived
          (bind _playerpawn (Variables|Default|GetPlayerPawn)
            (:then
              (bind _weapon_2 (Variables|Default|GetWeapon)
                (:then
                  (Variables|/AshtonGameplay/Gameplay/CaptainAmerica/Shield/Actors/B_Ashton_Shield_Weap.B_Ashton_Shield_Weap_C|SetbTryingToBlock true _weapon_2)
                  ;[1 0.6488 0.3272] Check if we should cancel Tac sprint
                  (Variables|/AshtonGameplay/Gameplay/CaptainAmerica/Shield/Actors/B_Ashton_Shield_Weap.B_Ashton_Shield_Weap_C|GetProjectile%20Replicated _weapon_2
                    (:then)
                    (:else
                      (Utilities|IsValid (Ability|GetFortAbilitySystemComponent _playerpawn)
                        (:Is%20Valid
                          (if (GameplayTags|HasMatchingGameplayTag (Ability|GetFortAbilitySystemComponent _playerpawn) "(TagName=\"Gameplay.Action.Player.TacticalSprint\")")
                            (Development|PrintString "Requesting server cancel Tac Sprint")
                            (Class|BAshtonShieldWeap|ServerCancelTacSprint _weapon_2) ; Let the server (via the weapon) know to cancel the tac sprint ability
                          ))))) ; Only if the projectile is not thrown
                )
                (:else)))
            (:else)))))))

(event Ability|EventOnEndAbility
  (params (bWasCancelled :type Boolean))
  (|CallParentFunction bWasCancelled)
  (Variables|Default|GetPlayerPawn
    (:then
      (bind _weapon (Variables|Default|GetWeapon)
        (:then
          (Variables|/AshtonGameplay/Gameplay/CaptainAmerica/Shield/Actors/B_Ashton_Shield_Weap.B_Ashton_Shield_Weap_C|SetbTryingToBlock false _weapon))
        (:else)))
    (:else)))

(event Ability|EventFailedToActivatePassiveAbility
  (params (ActorInfo :type Gameplay%20Ability%20Actor%20Info%20Structure))
  (|CallParentFunction))
