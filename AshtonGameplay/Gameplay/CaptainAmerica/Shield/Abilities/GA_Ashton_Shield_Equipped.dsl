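; Gameplay ability for the equipped Captain America shield that grants a custom tactical-sprint charge bash MME while the shield is in hand.
; It adds or removes the MME on the controller's TacticalSprintControllerComponent as the weapon's replicated projectile appears or clears, gated by a scalable-float toggle.
; Adding during an active sprint is deferred via a tag-count task until sprint stops, avoiding interruption.
; Ending the ability removes the MME and unbinds the event.
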(event Ability|EventActivateAbility
  ; Activate Ability
  (bind _outfortplayerpawn (CallFunction|GetAvatarasValidFortPlayerPawn)
    (:ValidFortPlayerPawn
      (Variables|Default|SetPlayerPawn _outfortplayerpawn)
      (bind _asb_ashton_shield_weap (Utilities|Casting|CastToB_Ashton_Shield_Weap (Weapon|GetCurrentWeapon (Variables|Default|GetPlayerPawn))))
      (bind _output_get (Variables|Default|SetPlayerWeapon _asb_ashton_shield_weap))
      (Default|BindEventtoOnShieldProjectileReplicated _output_get (EventDispatchers|CreateEvent :!event Custom|ShieldProjectileReplicated))
      (Utilities|FlowControl|Sequence
        (:then_0
          (Variables|/AshtonGameplay/Gameplay/CaptainAmerica/Shield/Actors/B_Ashton_Shield_Weap.B_Ashton_Shield_Weap_C|GetProjectile%20Replicated (Variables|Default|GetPlayerWeapon)
            (:then)
            (:else
              (if (GameplayTags|HasMatchingGameplayTag (Variables|Default|GetPlayerPawn) "(TagName=\"Gameplay.Action.Player.TacticalSprint\")")
                (Variables|Default|SetOnStopTacSprint_AddSprintChargeMME true)
                (else
                  (Default|AddSprintChargeMME))))) ; If the projectile is out (is valid),  don't add charge bash MME on equip
        )
        (:then_1
          ;;[1 1 1] If we are actively sprinting when the shield returns, don't add the charge MME until we stop sprinting so we don't interupt the sprint.
          (bind (_asynctaskproxy _tagcount) (Ability|Tasks|WaitGameplayTagCountChangedOnActor (Variables|Default|GetPlayerPawn) "(TagName=\"Gameplay.Action.Player.TacticalSprint\")")
            (:TagCountChanged
              (if (and (<= _tagcount 0) (Variables|Default|GetOnStopTacSprint_AddSprintChargeMME))
                (Variables|/AshtonGameplay/Gameplay/CaptainAmerica/Shield/Actors/B_Ashton_Shield_Weap.B_Ashton_Shield_Weap_C|GetProjectile%20Replicated (Variables|Default|GetPlayerWeapon)
                  (:then)
                  (:else
                    (Default|AddSprintChargeMME)
                    (Variables|Default|SetOnStopTacSprint_AddSprintChargeMME false)))))))))))

(event Custom|ShieldProjectileReplicated
  ; When the shield is thrown, remove our charge MME, when it returns, add it back - if we still have it equipped.
  (Variables|/AshtonGameplay/Gameplay/CaptainAmerica/Shield/Actors/B_Ashton_Shield_Weap.B_Ashton_Shield_Weap_C|GetProjectile%20Replicated (Variables|Default|GetPlayerWeapon)
    (:then
      (Default|RemoveSprintChargeMME))
    (:else
      (if (not (GameplayTags|HasMatchingGameplayTag (Variables|Default|GetPlayerPawn) "(TagName=\"Gameplay.Action.Player.TacticalSprint\")"))
        (Default|AddSprintChargeMME)
        (else
          (Variables|Default|SetOnStopTacSprint_AddSprintChargeMME true))))))

(event Ability|EventOnEndAbility
  (params (bWasCancelled :type Boolean))
  ; On End Ability
  (Variables|Default|SetOnStopTacSprint_AddSprintChargeMME false)
  (Default|RemoveSprintChargeMME)
  (bind _playerweapon (Variables|Default|GetPlayerWeapon)
    (:then
      (Default|UnbindEventfromOnShieldProjectileReplicated _playerweapon (EventDispatchers|CreateEvent :!event Custom|ShieldProjectileReplicated)))
    (:else)))

(fn AddSprintChargeMME
  (bind _playerpawn (Variables|Default|GetPlayerPawn))
  (bind _returnvalue (Pawn|GetController _playerpawn))
  (if (ScalableFloat|GetValueasBool (Variables|Default|GetHF_Charge_MME_Enabled))
    (Variables|Default|GetPlayerController
      (:then
        _)
      (:else
        (Utilities|IsValid _returnvalue
          (:Is%20Valid
            (Variables|Default|SetPlayerController _returnvalue)))))
    (Sprint|AddCustomSprintMMEClass (Actor|GetComponentbyClass (Variables|Default|GetPlayerController) "/TacticalSprintGame/Gameplay/TacticalSprintControllerComponent.TacticalSprintControllerComponent_C") (Variables|Default|GetCustomTacSprintExtensionClass))))

(fn RemoveSprintChargeMME
  (bind _playerpawn (Variables|Default|GetPlayerPawn))
  (bind _returnvalue (Pawn|GetController _playerpawn))
  (if (ScalableFloat|GetValueasBool (Variables|Default|GetHF_Charge_MME_Enabled))
    (Variables|Default|GetPlayerController
      (:then
        (Sprint|RemoveCustomSprintMMEClass (Actor|GetComponentbyClass (Variables|Default|GetPlayerController) "/TacticalSprintGame/Gameplay/TacticalSprintControllerComponent.TacticalSprintControllerComponent_C") (Variables|Default|GetCustomTacSprintExtensionClass)))
      (:else
        (Utilities|IsValid _returnvalue
          (:Is%20Valid
            (Variables|Default|SetPlayerController _returnvalue)))))))
