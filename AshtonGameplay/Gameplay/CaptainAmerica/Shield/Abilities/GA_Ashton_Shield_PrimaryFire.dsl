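; Gameplay ability for the primary fire of the Captain America shield, which triggers a shield throw.
; Activation commits the ability and ends it if the commit fails.
; On commit it caches the activating pawn and, unless that pawn already has the Throwing tag, sends the shield Throw gameplay event, then ends the ability.
; The throw itself is not implemented here, only the event that requests it.
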(event Ability|EventActivateAbility
  ;[1 1 1] Activate Ability
  (|CallParentFunction)
  (bind _returnvalue (Ability|CommitAbility))
  (if (not _returnvalue)
    (Ability|EndAbility)))

(event Ability|EventCommitExecute
  (bind _returnvalue (Ability|GetActivatingPawn))
  ;[1 1 1] Commit Execute
  (|CallParentFunction)
  (Utilities|IsValid _returnvalue
    (:Is%20Valid
      (bind _output_get (Variables|Default|SetPlayerPawn _returnvalue))
      (if (not (GameplayTags|HasMatchingGameplayTag _output_get "(TagName=\"Granted.Ashton.Shield.Throwing\")"))
        (Ability|SendGameplayEvent "(TagName=\"Event.Athena.Ashton.Shield.Throw\")"))
      (Ability|EndAbility))
    (:Is%20Not%20Valid
      (Ability|EndAbility))))

(event Ability|EventOnEndAbility
  (params (bWasCancelled :type Boolean))
  ;[1 1 1] On End Ability
  (|CallParentFunction bWasCancelled))

(fn OverrideFailedReason
  (params (FailedReason :type Gameplay%20Tag%20Container%20Structure%20%28by%20ref%29))
  (return 0))
