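; Gameplay ability for throwing Captain America's shield as a homing projectile.
; On commit it plays the throw montage, whose notify fires cues and, on the server, disables shield projectile blocking and spawns the projectile with an explosion effect container.
; It copies the weapon's homing target to the projectile and applies a wait-for-catch effect.
; Spawn location uses stance-based offsets, with line traces to lift it over low obstructions and avoid blocking geometry.
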(event Ability|EventActivateAbility
  ;[1 1 1] Activate Ability
  (|CallParentFunction)
  (bind _returnvalue (Ability|CommitAbility))
  (if (not _returnvalue)
    (Ability|EndAbility)))

(event Ability|EventCommitExecute
  (bind _returnvalue (Ability|GetActivatingPawn))
  ;[1 1 1] Commit Execute
  (|CallParentFunction)
  (Utilities|IsValid _returnvalue
    (:Is%20Valid
      (bind _output_get (Variables|ReferencedActors|SetPlayerPawn _returnvalue))
      (bind _asb_ashton_shield_weap (Utilities|Casting|CastToB_Ashton_Shield_Weap (Weapon|GetCurrentWeapon _output_get))
        (:then
          (Variables|Default|SetShieldWeapon _asb_ashton_shield_weap)
          (CallFunction|ThrowProjectile))
        (:CastFailed
          (Ability|EndAbility))))
    (:Is%20Not%20Valid
      (Ability|EndAbility))))

(event Custom|ThrowProjectile
  ;[1 1 1] Throw Projectile
  (Ability|Tasks|PlayMontageAndWaitNotify "None" (Variables|Default|GetMontage_Throw) 1.0 "None" false
    (:then
      (Ability|ExecuteGameplayCueOnOwner (Variables|Default|GetGC_Throw_Foley_SFX))
      ; Fail-safe to end ability if montage events aren't called
      (Utilities|FlowControl|Delay (Animation|GetPlayLength (Variables|Default|GetMontage_Throw)))
      (Ability|EndAbility))
    (:OnCompleted
      _)
    (:OnInterrupted
      _)
    (:OnCancelled
      _)
    (:OnNotifyBegin
      (CallFunction|SpawnProjectile)))
  (Ability|EndAbility))

(event Ability|EventOnEndAbility
  (params (bWasCancelled :type Boolean))
  ;[1 1 1] On End Ability
  (|CallParentFunction bWasCancelled))

(event Custom|SpawnProjectile
  (bind _shield_projectile (Variables|Default|GetShield%20Projectile))
  ;[1 1 1] Spawn Projectile!
  (Ability|ExecuteGameplayCueOnOwner (Variables|Default|GetGC_Throw_VFX))
  (if (Networking|IsServer)
    (Utilities|FlowControl|Sequence
      (:then_0
        (bind _shieldweapon (Variables|Default|GetShieldWeapon)
          (:then
            (bind _blockingbga (Variables|/AshtonGameplay/Gameplay/CaptainAmerica/Shield/Actors/B_Ashton_Shield_Weap.B_Ashton_Shield_Weap_C|GetBlockingBGA _shieldweapon)
              (:then
                (Class|BGAAshtonShieldBlockVolume|ToggleBlockProjectiles _blockingbga false)
                (Variables|/AshtonGameplay/Gameplay/CaptainAmerica/Shield/Actors/BGA_Ashton_Shield_BlockVolume.BGA_Ashton_Shield_BlockVolume_C|SetbActiveBlocking false _blockingbga))
              (:else)))
          (:else)))
      (:then_1
        (Variables|Default|SetExplosionEffectContainerSpec (Ability|MakeGameplayEffectContainerSpec (Variables|Default|GetECDefaultExplosion)))
        (bind _shieldweapon_2 (Variables|Default|GetShieldWeapon)
          (:then
            (Homing|UpdateHomingCurrentTarget _shieldweapon_2)
            (bind (_location _rotation) (Default|CalculatePrjSpawnLocationandRotation))
            (Variables|Default|SetSpawnDirection (Math|Rotator|CombineRotators _rotation (Variables|Default|GetModifyProjectileDirection)))
            (Variables|Default|SetInitialSpeed (ScalableFloat|GetValueatLevel (Variables|Default|GetProjectileSpeed)))
            (bind (_asynctaskproxy _projectiledata) (Projectile|Tasks|SpawnProjectileAndWait :Class "/AshtonGameplay/Gameplay/CaptainAmerica/Shield/Actors/B_Prj_Ashton_Shield.B_Prj_Ashton_Shield_C" :RequestedBy (Variables|ReferencedActors|GetPlayerPawn) :SpawnLocation _location :SpawnRotation _rotation :SpawnDirection (Variables|Default|GetSpawnDirection) :InitialSpeed (Variables|Default|GetInitialSpeed) :EffectContainerSpecToApplyOnExplode (Variables|Default|GetExplosionEffectContainerSpec) :GravityScale 0.0 :bAllowProjectileSpawnWhenDead false :bAllowProjectileSpawnWhenDBNO false :ReturnVelocity (Math|Vector|NegateVector (* (Variables|Default|GetInitialSpeed) (Math|Vector|GetForwardVector (Variables|Default|GetSpawnDirection)))) :PreviousVelocity (* (Variables|Default|GetInitialSpeed) (Math|Vector|GetForwardVector (Variables|Default|GetSpawnDirection)))))
            (bind _projectiledata_spawnedprojectile (Utilities|Struct|BreakProjectileEventData _projectiledata))
            (bind _output_get (Variables|ReferencedActors|SetSpawnedProjectile _projectiledata_spawnedprojectile))
            (Ability|ExecuteGameplayCueWithParamsOnOwner (Variables|Default|GetGC_Throw_SFX) (Ability|GameplayCue|MakeGameplayCueParameters :NormalizedMagnitude 0.0 :RawMagnitude 0.0 :Location "0,0,0" :Normal "0,0,0" :Instigator _output_get :GameplayEffectLevel 1 :AbilityLevel 1 :bReplicateLocationWhenUsingMinimalRepProxy false))
            (bind _asb_prj_ashton_shield (Utilities|Casting|CastToB_Prj_Ashton_Shield _output_get))
            (Variables|Default|SetShield%20Projectile _asb_prj_ashton_shield)
            (Utilities|FlowControl|Sequence
              (:then_0
                ;[1 1 1] Set the homing target based on the target passed in from the GameplayEvent
                (bind _shieldweapon_3 (Variables|Default|GetShieldWeapon)
                  (:then
                    (Homing|UpdateHomingCurrentTarget _shieldweapon_3) ; Even though we run this earlier, we need to re-run it again after the latent node
                    (bind _homingtarget (Variables|/AshtonGameplay/Gameplay/CaptainAmerica/Shield/Actors/B_Ashton_Shield_Weap.B_Ashton_Shield_Weap_C|GetHomingTarget _shieldweapon_3)
                      (:then
                        (Variables|/AshtonGameplay/Gameplay/CaptainAmerica/Shield/Actors/B_Prj_Ashton_Shield.B_Prj_Ashton_Shield_C|SetHomingTarget _homingtarget _shield_projectile)
                        (Default|SetHomingVars _shield_projectile false))
                      (:else)))
                  (:else)))
              (:then_1
                ;[1 1 1] Send Event to track projectile. Ability will end when throw montage is complete
                (Ability|ApplyGameplayEffectToOwner (Variables|Default|GetGE%20Wait%20Catch)))))
          (:else))))))

(fn CalculatePrjSpawnLocationAndRotation
  (bind (_outlocation _outrotation) (Actor|GetActorEyesViewPoint (Variables|ReferencedActors|GetPlayerPawn)))
  (bind _verbosedebug (Variables|Default|GetVerboseDebug))
  (bind _outloc (LocalVariables|GetOutLoc))
  (Utilities|FlowControl|Sequence
    (:then_0
      (LocalVariables|SetOutLoc (+ (Transformation|GetActorLocation (Variables|ReferencedActors|GetPlayerPawn)) (Math|Vector|vector+vector (* (Math|Vector|GetForwardVector (Pawn|GetControlRotation (Variables|ReferencedActors|GetPlayerPawn) :!class /Script/Engine.Pawn)) (Variables|PrjSpawnOffsets|GetSpawnForwardOffset)) (* (Math|Vector|GetRightVector (Pawn|GetControlRotation (Variables|ReferencedActors|GetPlayerPawn) :!class /Script/Engine.Pawn)) (Variables|PrjSpawnOffsets|GetSpawnRightOffset)) (Math|Vector|SelectVector (* (Math|Vector|GetUpVector (Pawn|GetControlRotation (Variables|ReferencedActors|GetPlayerPawn) :!class /Script/Engine.Pawn)) (Variables|PrjSpawnOffsets|GetSpawnUpOffsetRiding)) (select (Variables|/Script/Engine.Character|GetbIsCrouched (Variables|ReferencedActors|GetPlayerPawn)) (* (Math|Vector|GetUpVector (Pawn|GetControlRotation (Variables|ReferencedActors|GetPlayerPawn) :!class /Script/Engine.Pawn)) (Variables|PrjSpawnOffsets|GetSpawnUpOffsetCrouch)) (* (Math|Vector|GetUpVector (Pawn|GetControlRotation (Variables|ReferencedActors|GetPlayerPawn) :!class /Script/Engine.Pawn)) (Variables|PrjSpawnOffsets|GetSpawnUpOffset))) (GameplayTags|HasAnyMatchingGameplayTags (Variables|ReferencedActors|GetPlayerPawn) (LocalVariables|GetRidingOrInVehicleTags))))))
      (LocalVariables|SetOutRot _outrotation)
      (Variables|ReferencedActors|GetPlayerPawn
        (:then
          (if (ScalableFloat|GetValueasBool (Variables|ScalableFloats|GetROW_OffsetSpawnVertically_WhenObstructedLow))
            ; FORT-769517: Adjust the spawn location vertically, if there are obstructions in the path of the projectile. This puts the spawn point at the reticle, so it will be true to aim in these scenarios. Normally it spawns below the reticle to keep the reticle visually unobstructed. But this can cause you to hit ledges and other items below the reticle in certain cases
            (bind _shieldweapon (Variables|Default|GetShieldWeapon)
              (:then
                (bind _homingtarget (Variables|/AshtonGameplay/Gameplay/CaptainAmerica/Shield/Actors/B_Ashton_Shield_Weap.B_Ashton_Shield_Weap_C|GetHomingTarget _shieldweapon)
                  (:then
                    (LocalVariables|SetEndTraceLocationCollisionBelowReticle (Transformation|GetActorLocation _homingtarget)))
                  (:else
                    (LocalVariables|SetEndTraceLocationCollisionBelowReticle (+ (LocalVariables|GetOutLoc) (* (Math|Vector|GetForwardVector (Pawn|GetControlRotation (Variables|ReferencedActors|GetPlayerPawn) :!class /Script/Engine.Pawn)) (ScalableFloat|GetValueatLevel (Variables|ScalableFloats|GetROW_OffsetSpawnVertically_LineTraceDistance)))))))
                (bind (_outhit _returnvalue) (Collision|LineTraceByChannel :Start (LocalVariables|GetOutLoc) :End (LocalVariables|GetEndTraceLocationCollisionBelowReticle) :TraceChannel "TraceTypeQuery1" :bTraceComplex false :DrawDebugType "ForDuration" :bIgnoreSelf true :TraceColor "(R=0.225796,G=0.692822,B=1.000000,A=1.000000)" :TraceHitColor "(R=0.627098,G=0.034841,B=1.000000,A=1.000000)" :DrawTime 5.0))
                (if _returnvalue
                  (Utilities|Casting|CastToFortPawn (.HitActor (Collision|BreakHitResult _outhit))
                    (:then
                      _)
                    (:CastFailed
                      (LocalVariables|SetOutLoc (+ (LocalVariables|GetOutLoc) (* (Math|Vector|GetUpVector (Pawn|GetControlRotation (Variables|ReferencedActors|GetPlayerPawn) :!class /Script/Engine.Pawn)) (ScalableFloat|GetValueatLevel (Variables|ScalableFloats|GetROW_OffsetSpawnVertically_Offset)))))))
                  (bind _playerpawn (Variables|ReferencedActors|GetPlayerPawn)
                    (:then
                      (bind (_outhit_2 _returnvalue_2) (Collision|LineTraceByProfile (Transformation|GetActorLocation _playerpawn) _outloc "FortProjectileThrownOverlapPawns" false (Utilities|Array|MakeArray _playerpawn (Variables|/Script/FortniteGame.FortPawn|GetCurrentWeapon _playerpawn) (Actor|GetAttachParentActor _playerpawn)) "ForDuration" true "(R=1.000000,G=0.000000,B=0.000000,A=1.000000)" "(R=0.000000,G=1.000000,B=0.000000,A=1.000000)" 1e+01)) ; Do NOT worry about pawns  here. The projectile has a much larger pawn specific collision,\nso this issue does not occur when throwing at a close pawn. 
                      (if _returnvalue_2
                        (LocalVariables|SetOutLoc (+ (* (Math|Vector|Normalize (- (Transformation|GetActorLocation _playerpawn) _outloc)) (Variables|PrjSpawnOffsets|GetAdjustmentHitOffset)) (.Location (Collision|BreakHitResult _outhit_2))))
                        (Development|PrintString (Utilities|String|Append "Start location trace - Blocking hit, adjusting start location - " (Utilities|GetDisplayName (.HitActor (Collision|BreakHitResult _outhit_2)))))
                        (else
                          (Utilities|IsValid (.HitActor (Collision|BreakHitResult _outhit_2))
                            (:Is%20Valid
                              (Development|PrintString (Utilities|String|Append "Start location trace - non blocking hit - " (Utilities|GetDisplayName (.HitActor (Collision|BreakHitResult _outhit_2)))) _verbosedebug _verbosedebug))
                            (:Is%20Not%20Valid
                              (Development|PrintString "Start location trace - No hit" _verbosedebug _verbosedebug))))))
                    (:else))
                  (else
                    _)))
              (:else))
            (else
              _)))
        (:else)))
    (:then_1
      ;[1 0.6073 0.3726] Return OutLoc and OutRot
      (return (LocalVariables|GetOutLoc) (LocalVariables|GetOutRot)))))
